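# User widget wrapping an AudioMaterialKnob that edits a MetaSound float literal through its viewmodel.
# Knob changes map the normalized value onto the viewmodel's source range, optionally logarithmically, and set SourceValue; a flag suppresses this until the tick after Construct.
# The input viewmodel is looked up by Input Name from the supplied map.
# PreConstruct applies Knob Style, mouse capture events are re-broadcast, and mouse enter/leave field-notify Is Hovered.
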WidgetBlueprint /TechAudioToolsContent/Widgets/MetaSoundLiterals/Float/Primitives/WBP_MetaSoundLiteralFloatPrimitive_Knob
graphs [
 EventGraph {nodes [
   n1 = "On Knob Value Changed (AudioMaterialKnob)"!() -> (OutputDelegate, then, Value:float) {is=title K2Node_ComponentBoundEvent}
   > n2 = Get!(execute:Exec) -> (then:Exec, else:Exec, Is Constructing:bool)
   n3 = Get MetaSoundLiteralViewModel_Float() -> (MetaSoundLiteralViewModel_Float:Meta Sound Literal Float Viewmodel Object Reference)
   Set SourceValue!(
    execute=@n2.else
    SourceValue:float=Get Source Value!(
     Apply Logarithmic Interpolation:bool=Get Apply Logarithmic Interpolation() -> (Apply Logarithmic Interpolation:bool)
     Normalized Value:double=@n1.Value
     Source Range Min:double=GetSourceRangeMin!(self:Meta Sound Literal Float Viewmodel Object Reference=@n3) -> (ReturnValue:float)
     Source Range Max:double=GetSourceRangeMax!(self:Meta Sound Literal Float Viewmodel Object Reference=@n3) -> (ReturnValue:float)
    ) -> (Source Value:double) {is=title K2Node_CallFunction}
    self:Meta Sound Literal Float Viewmodel Object Reference=Get MetaSoundLiteralViewModel_Float()
   ) -> (then, Output_Get:float)
   "On Mouse Capture Begin (AudioMaterialKnob)"() {is=title K2Node_ComponentBoundEvent}
   > Call On Mouse Capture Begin!(execute:Exec, self:Meta Sound Literal Float Primitive Material Knob Object Reference) -> (then:Exec) {is=title K2Node_CallDelegate}
   "On Mouse Capture End (AudioMaterialKnob)"() {is=title K2Node_ComponentBoundEvent}
   > Call On Mouse Capture End!(execute:Exec, self:Meta Sound Literal Float Primitive Material Knob Object Reference) -> (then:Exec) {is=title K2Node_CallDelegate}
   Event Construct()
   > Set Is Constructing(Is Constructing:bool=true) -> (Output_Get:bool)
   > DelayUntilNextTick!(execute) -> (then)
   > Set Is Constructing(Is Constructing=false)
   n4 = Event OnMouseEnter()
   Broadcast Field Value Changed!(execute=[@n4.then, @n5.then], self:Widget Object Reference, FieldId:Field Notification Id Structure=(FieldName="Is Hovered")) -> (then)
   n5 = Event OnMouseLeave!() -> (OutputDelegate, then, MouseEvent:"Pointer Event Structure (by ref)") {is=title K2Node_Event}
   Event PreConstruct()
   > Set WidgetStyle!(
    execute
    WidgetStyle:Audio Material Knob Style Structure=Get Knob Style() -> (Knob Style:Audio Material Knob Style Structure)
    self:Audio Material Knob Object Reference=Get AudioMaterialKnob() -> (AudioMaterialKnob:Audio Material Knob Object Reference)
   ) -> (then, Output_Get:Audio Material Knob Style Structure)
   n6 = Event Set Input Viewmodels!() -> (OutputDelegate, then, InputViewModels:Map of Names to Meta Sound Input Viewmodel Object References) {is=title K2Node_Event}
   > SetMetaSoundInputViewModel!(
    execute
    self
    Viewmodel:Meta Sound Input Viewmodel Object Reference=Find(
     TargetMap:Map of Names to Meta Sound Input Viewmodel Object References=@n6.InputViewModels
     Key:"Name (by ref)"=Get Input Name() -> (Input Name:Name=None)
    ) -> (Value:Meta Sound Input Viewmodel Object Reference)
   ) -> (then)
  ]}
 GetInputViewModelNames {nodes [
   Get Input Viewmodel Names() {is=title K2Node_FunctionEntry}
   > Return Node!(execute, ReturnValue:Array of Names=Make Array("[0]":Name=Get Input Name() -> (Input Name=None)) -> (Array:Array of Names))
  ]}
 Is Hovered {nodes [
   Is Hovered() {is=title K2Node_FunctionEntry}
   > Return Node(ReturnValue:bool=IsHovered(self=Get AudioMaterialKnob()))
  ]}
 On Mouse Capture Begin {nodes [On Mouse Capture Begin() {is=title K2Node_FunctionEntry}]}
 On Mouse Capture End {nodes [On Mouse Capture End() {is=title K2Node_FunctionEntry}]}
]
variables [
 Apply Logarithmic Interpolation {type=boolean}
 Input Name {type=string, default=None}
 Is Constructing {type=boolean}
 Knob Style {
  type=AudioMaterialKnobStyle
  default {
   material=null
   desiredSize=<64 64>
   knobAccentColor=<rgba 0.06 0.06 0.06 1>
   knobBarColor=<rgba 0.067 0.067 0.067 1>
   knobBarFillMaxColor=<rgba 0.96 0.96 0.96 1>
   knobBarFillMidColor=<rgba 0.96 0.96 0.96 1>
   knobBarFillMinColor=<rgba 0.96 0.96 0.96 1>
   knobBarFillTintColor=<rgba 0.96 0.96 0.96 1>
   knobBarShadowColor=<rgba 0.067 0.067 0.067 1>
   knobEdgeFillColor=<rgba 0.0156 0.0156 0.0156 1>
   knobIndicatorDotColor=<rgba 1 0 0 1>
   knobMainColor=<rgba 0.1406 0.1406 0.1406 1>
   knobShadowColor=<rgba 0.06 0.06 0.06 1>
   knobSmoothBevelColor=<rgba 0.0417 0.0417 0.0417 1>
   textBoxStyle {
    backgroundColor {colorUseRule=UseColor_ColorTable, specifiedColor=<rgba 0 0 0 0>}
    backgroundImage {
     bIsDynamicallyLoaded=false
     drawAs=RoundedBox
     imageType=NoImage
     mirroring=NoMirror
     resourceName=None
     resourceObject=null
     tiling=NoTile
     imageSize=<56 28>
     margin=<ltrb 0 0 0 0>
     outlineSettings {
      bUseBrushTransparency=false
      roundingType=FixedRadius
      width=0
      color {colorUseRule=UseColor_Specified, specifiedColor=<rgba 0 0 0 0>}
      cornerRadii=<4 4 4 4>
     }
     tintColor {colorUseRule=UseColor_ColorTable, specifiedColor=<rgba 0 0 0 0>}
     uVRegion {bIsValid=false, max=<0 0>, min=<0 0>}
    }
   }
  }
 }
 Object To Transact {type=Object, default=null}
 On Mouse Capture Begin {}
 On Mouse Capture End {}
]
widgets {
 info {inheritedWidgetCount=0, parentClass=&UserWidget, rootWidgetClass=/Script/AudioWidgets.AudioMaterialKnob, widgetCount=1}
 widgets [AudioMaterialKnob {widgetClassPath=&AudioMaterialKnob, bIsVariable=true, widget=AudioMaterialKnob}]
}
